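# Niagara system for the MasakoReign emote that draws a fiery dragon over roughly 4.6 seconds.
# All seven emitters sample bones through a shared skeletal mesh reader, forming a ribbon body and fire trail with curved flame meshes, smoke and sparks.
# A head mesh with a light and flame shell follows the sampled position, the flame feeding smoothed position to its material.
# Emitter-age curves grow the ribbons in and shrink or fade elements near the end.
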NiagaraSystem /BRCosmetics/Effects/Fort_Effects/Emotes/MasakoReign/NS_Emote_MasakoReign_Dragon
dependencies {
 usedDataInterfaces [&NiagaraDataInterfaceCurve, &NiagaraDataInterfaceColorCurve, &NiagaraDataInterfaceSkeletalMesh]
 usedDynamicInputs [&FNTimeOfDayScalar]
 usedModules [
  &self:SystemSpawnScript.SystemScriptSource.SystemScriptGraph.NiagaraNodeAssignment_0.SetVariables_#1
  &self:SystemSpawnScript.SystemScriptSource.SystemScriptGraph.NiagaraNodeAssignment_4.SetVariables_#2
  &SystemState
  &self:SystemSpawnScript.SystemScriptSource.SystemScriptGraph.NiagaraNodeAssignment_1.SetVariables_#3
  &self:SystemSpawnScript.SystemScriptSource.SystemScriptGraph.NiagaraNodeAssignment_2.SetVariables_#4
  &self:SystemSpawnScript.SystemScriptSource.SystemScriptGraph.NiagaraNodeAssignment_3.SetVariables_#5
 ]
}
emitters {
 CurveFire {
  emitterUpdateScript [
   &EmitterState(Inactive Response=::NewEnumerator1, Life Cycle Mode=::NewEnumerator1, Loop Behavior=::NewEnumerator1, Loop Duration=4.6, UseLoopDelay=false)
   &SpawnRate(SpawnRate=20, Use Spawn Probability=true)
   &SpawnBurst_Instantaneous(Use Spawn Probability=true, UseLoopCountLimit=false) {enabled=false}
   /SetVariables(Emitter.Allsize:NiagaraFloat=&RandomRangeFloat(Minimum=0.5, Maximum=0.9)) {isSetParameters=true}
  ]
  particleSpawnScript [
   &InitializeParticle(
    Color=<rgba 0.6 0.1466 0.1389 5>
    Lifetime Max=1
    Lifetime Min=0.7
    Lifetime Mode=::NewEnumerator1
    Mesh Scale=<0.5 0.5 0.5>
    Mesh Scale Mode=::NewEnumerator4
    Position Offset=&RandomRangeVector(Minimum=<0 0 -25>, Maximum=<0 0 20>)
    UsePositionOffset=true
    Write Material Random=true
   )
   &SkeletalMeshLocation(
    Bone Sampling Mode=::NewEnumerator4
    MaskSpawnGroup=false
    Sampled Position Offset Coordinate Space:ENiagaraSkelMeshTransforms=::NewEnumerator2
    Skeletal Mesh=$System.Skeletal Mesh Reader
   )
   &ShapeLocation(MaskSpawnGroup=false, Override Seed=false, Sphere Radius=10, UseOwnerScale=true)
   &InitialMeshOrientation(Mesh Orientation Mode=::NewEnumerator7, Rotation=<0 0 0.25>, Use Rotation Vector=true)
   &InitialMeshOrientation(
    Mesh Orientation Mode=::NewEnumerator2
    Rotation=&RandomRangeVector(Minimum=<-1 0 0>, Maximum=<1 0 0>)
    Use Rotation Vector=true
   ) {module=InitialMeshOrientation001}
   &AddVelocity(Module.UseConeFalloff=false, Module.UseVelDistribution=false, Override Seed=false, Velocity Mode=::NewEnumerator2, Velocity Speed=200) {enabled=false}
  ]
  particleUpdateScript [
   /SetVariables(Particles.AlphaAll:NiagaraFloat=&FloatFromCurve(
     FloatCurve=DataInterface() {propertyValues {Curve {defaultValue=1, keys [{tangentMode=RCTM_User, time=0, value=0}, {tangentMode=RCTM_User, time=0.5, value=1}]}}}
     Scale Curve=&FloatFromCurve(
      FloatCurve=DataInterface() {propertyValues {Curve {keys [{arriveTangent=-2.8607, interpMode=RCIM_Cubic, leaveTangent=-2.8607, tangentMode=RCTM_User, time=0.0044, value=0.9529}, <at 1 0>]}}}
      Scale Curve=&Multiply_Float(A=1.5, B=1)
     )
    )) {isSetParameters=true}
   /SetVariables(Particles.BrightnessAll:NiagaraFloat=&FloatFromCurve(
     FloatCurve=DataInterface() {propertyValues {Curve {keys [
         <at 0 -0.0118 RCIM_Cubic>
         {arriveTangent=0.3618, interpMode=RCIM_Cubic, leaveTangent=0.3618, time=0.2035, value=1}
         {arriveTangent=-1.2554, interpMode=RCIM_Cubic, leaveTangent=-1.2554, time=0.5284, value=0.1794}
         <at 1 0>
        ]}}}
     Scale Curve=&MakeCustomFloatFromBool(Bool=&RandomBool(), True Float=0.3, False Float=1)
    )) {isSetParameters=true}
   &DynamicMaterialParameters(
    Index 0 Param 1=&Multiply_Float(A=1.1, B=$Emitter.Allsize)
    Index 0 Param 2=2
    Index 0 Param 3=&RandomRangeFloat(Minimum=0, Maximum=1)
    Index 0 Param 4=0.5
    Index 1 Param 1=&FloatFromCurve(FloatCurve=DataInterface() {propertyValues {Curve {keys [<at 0 0>,<at 1 1>]}}})
    Index 1 Param 2=0.3
    Module.Index0Float0Write=true
    Module.Index0Float1Write=true
    Module.Index0Float2Write=true
    Module.Index0Float3Write=true
    Module.Index1Float0Write=true
    Module.Index1Float1Write=true
    Module.Index1Float2Write=false
    Module.Index1Float3Write=false
    Write Parameter Index 1=::NewEnumerator1
   )
   &ParticleState()
   &UpdateMeshOrientation(Rotation Rate=-0.2, Rotation Vector=<0 1 0>)
   &ScaleMeshSize(Scale Factor=&Multiply_VectorByFloat(
     Vector=&MakeCustomVectorFromBool(Bool=&RandomBool(), True Vector=<-1 1 1>, False Vector=<1 1 1>)
     Float=&FloatFromCurve(
      FloatCurve=DataInterface() {propertyValues {Curve {
         defaultValue=1
         keys [
          {arriveTangent=1.0428, interpMode=RCIM_Cubic, time=0, value=0}
          {arriveTangent=-0.2, interpMode=RCIM_Cubic, leaveTangent=-0.2, tangentMode=RCTM_Break, time=0.1997, value=1.1761}
          {interpMode=RCIM_Cubic, leaveTangent=0.8255, time=1, value=1}
         ]
        }}}
      Scale Curve=&Multiply_Float(A=&RandomRangeFloat(Minimum=0.8, Maximum=0.8), B=$Emitter.Allsize)
     )
    ))
   &Drag(Use Linear Drag=true, Use Rotational Drag=true)
   &CurlNoiseForce(Bias Noise=false, Coordinate Space:ENiagaraCoordinateSpace=::Local, Noise Frequency=3, Noise Strength=500, Pan Noise=false, Random Seed=25) {enabled=false}
   &SolveForcesAndVelocity(Clamp Velocity=false, Limit Acceleration=false)
   &ScaleColor(
    Scale Alpha=$Particles.AlphaAll
    Scale RGB=&Multiply_VectorByFloat(
     Vector=&MakeVectorFromLinearColorRGB(Color=&ColorFromCurve(ColorCurve=$System.Curve for Colors))
     Float=$Particles.BrightnessAll
    )
    ScaleA=true
    ScaleRGB=true
   )
   &FN_Mesh_Attributes(
    Color=&MakeLinearColorFromVectorAndFloat(
     "Vector (RGB)"=&Multiply_Vector(
      A=&Multiply_VectorByFloat(Vector=&MakeVectorFromLinearColorRGB(Color=<rgba 5 0.4167 0 0.3>), Float=$Particles.BrightnessAll)
      B=&MakeVectorFromLinearColorRGB(Color=&ColorFromCurve(ColorCurve=$System.Curve for Colors))
     )
     "Float (Alpha)"=&Multiply_Float(A=0.3, B=$Particles.AlphaAll)
    )
    Create Color Attribute=true
    Module Output=::NewEnumerator1
   )
   &CameraOffset(Camera Offset Amount=$System.CameraOffsetAll)
  ]
  rendererClasses [&NiagaraMeshRendererProperties]
  renderers [
   {rendererClass=&NiagaraMeshRendererProperties, rendererIndex=0}
   {rendererClass=&NiagaraMeshRendererProperties, rendererIndex=1}
  ]
 }
 Dragon {
  emitterUpdateScript [
   &EmitterState(Inactive Response=::NewEnumerator1, Life Cycle Mode=::NewEnumerator1, Loop Behavior=::NewEnumerator1, Loop Duration=4.6, UseLoopDelay=false)
   &SpawnRate(SpawnRate=40, Use Spawn Probability=false)
  ]
  particleSpawnScript [
   &InitializeParticle(
    Lifetime=0.6
    Ribbon Width=36
    Ribbon Width Mode=::NewEnumerator1
    Sprite Size Mode=::NewEnumerator3
    Uniform Sprite Size=15
    UsePositionOffset=false
    Write Material Random=true
   )
   &SkeletalMeshLocation(
    Bone Sampling Mode=::NewEnumerator4
    MaskSpawnGroup=false
    Skeletal Mesh=$System.Skeletal Mesh Reader
    Use Sampled Mesh Attribute as Ribbon ID=false
   ) {module=SkeletalMeshLocation001}
  ]
  particleUpdateScript [
   &ParticleState()
   &ScaleColor(Scale RGB=&Multiply_VectorByFloat(Vector=<1 1 1>, Float=1.5), ScaleA=true, ScaleRGB=true)
   &ScaleRibbonWidth(Ribbon Width Scale=&Multiply_Float(
     A=&FloatFromCurve(
      FloatCurve=DataInterface() {propertyValues {Curve {keys [<at 0 1 RCIM_Cubic>, {arriveTangent=-2.0569, tangentMode=RCTM_User, time=1, value=0.3}]}}}
      Scale Curve=&FN_LOD_Distance_Remap(Start Distance=300, End Distance=400, Minimum Output=0.8, Maximum Output=1, Clamp Output=true)
     )
     B=&FloatFromCurve(FloatCurve=DataInterface() {propertyValues {Curve {keys [<at 0.5 0>,<at 0.8 1>]}}}, CurveIndex=$Emitter.Age)
    ))
   &CameraOffset(Camera Offset Amount=$System.CameraOffsetAll) {enabled=false}
   &DynamicMaterialParameters(
    Index 0 Param 2=&FloatFromCurve(FloatCurve=DataInterface() {propertyValues {Curve {keys [<at 0 0>,<at 3.9 0>,<at 4.7 0.8>]}}}, CurveIndex=$Emitter.Age)
    Index 0 Param 3=$System.CameraOffsetAll
    Module.Index0Float0Write=false
    Module.Index0Float1Write=true
    Module.Index0Float2Write=true
    Module.Index0Float3Write=true
   )
  ]
  rendererClasses [&NiagaraRibbonRendererProperties]
  renderers [{rendererClass=&NiagaraRibbonRendererProperties}]
 }
 DragonFireTrail {
  emitterUpdateScript [
   &EmitterState(Inactive Response=::NewEnumerator1, Life Cycle Mode=::NewEnumerator1, Loop Behavior=::NewEnumerator1, Loop Duration=4.6, UseLoopDelay=false)
   &SpawnRate(SpawnRate=40, Use Spawn Probability=false)
  ]
  particleSpawnScript [
   &InitializeParticle(
    Color=<rgba 0.2708 0 0 1>
    Lifetime=0.8
    Ribbon Width=40.5
    Ribbon Width Mode=::NewEnumerator1
    UsePositionOffset=false
    Write Material Random=true
   )
   &SkeletalMeshLocation(
    Bone Sampling Mode=::NewEnumerator4
    MaskSpawnGroup=false
    Sampled Position Offset=<0 0 -12>
    Skeletal Mesh=$System.Skeletal Mesh Reader
    Use Sampled Mesh Attribute as Ribbon ID=false
   ) {module=SkeletalMeshLocation001}
   &DynamicMaterialParameters(
    Index 1 Param 4=1
    Module.Index0Param0Write=true
    Module.Index0Param1Write=true
    Module.Index0Param2Write=true
    Module.Index0Param3Write=true
    Module.Index1Param0Write=true
    Module.Index1Param1Write=true
    Module.Index1Param2Write=true
    Module.Index1Param3Write=true
    Module.Index2Param0Write=true
    Module.Index2Param1Write=true
    Module.Index2Param2Write=true
    Module.Index2Param3Write=true
    Write Parameter Index 1=true
    Write Parameter Index 2=true
   ) {enabled=false}
  ]
  particleUpdateScript [
   /SetVariables(Particles.ScaleAlphaAll:NiagaraFloat=&FloatFromCurve(FloatCurve=DataInterface() {propertyValues {Curve {
        defaultValue=1
        keys [
         {interpMode=RCIM_Cubic, tangentMode=RCTM_User, time=0, value=0}
         {arriveTangent=2, interpMode=RCIM_Cubic, leaveTangent=2, time=0.2, value=1}
         {interpMode=RCIM_Cubic, tangentMode=RCTM_User, time=0.5, value=1}
         {tangentMode=RCTM_User, time=1, value=0}
        ]
       }}})) {isSetParameters=true}
   &ParticleState()
   &ScaleColor(Scale RGB=&Multiply_VectorByFloat(Vector=<1 1 1>, Float=1), ScaleA=true, ScaleRGB=true)
   &ScaleRibbonWidth(Ribbon Width Scale=&FloatFromCurve(
     FloatCurve=DataInterface() {propertyValues {Curve {keys [<at 0 1>, {arriveTangent=-2.0569, time=1, value=1}]}}}
     Scale Curve=&FloatFromCurve(FloatCurve=DataInterface() {propertyValues {Curve {keys [<at 0.5 0>,<at 0.8 1>]}}}, CurveIndex=$Emitter.Age)
    ))
   /SetVariables(Particles.Color3:LinearColor=&MultiplyLinearColor(
     Linear Color=<rgba 1.5 1.5 1.5 0.8>
     Scale Factor=&MakeVector4_FromFloat(Float=$Particles.ScaleAlphaAll)
    )) {isSetParameters=true}
   /SetVariables(Particles.Color4:LinearColor=&MultiplyLinearColor(
     Linear Color=<rgba 1 0.1406 0.2122 0.2>
     Scale Factor=&MakeVector4_FromFloat(Float=$Particles.ScaleAlphaAll)
    )) {isSetParameters=true}
   /SetVariables(Particles.Ribbon02:NiagaraFloat=&Multiply_Float(A=$Particles.RibbonWidth, B=0.7)) {isSetParameters=true}
   /SetVariables(Particles.Ribbon03:NiagaraFloat=&Multiply_Float(A=$Particles.RibbonWidth, B=1)) {isSetParameters=true}
   &CurlNoiseForce(Bias Noise=false, Coordinate Space:ENiagaraCoordinateSpace=::Local, Noise Frequency=3, Noise Strength=100, Pan Noise=false, Random Seed=25)
   &SolveForcesAndVelocity(Clamp Velocity=false, Limit Acceleration=false)
   &CameraOffset(Camera Offset Amount=$System.CameraOffsetAll) {enabled=false}
   &DynamicMaterialParameters(
    Index 0 Param 3=$System.CameraOffsetAll
    Index 1 Param 3=&FloatFromCurve(FloatCurve=DataInterface() {propertyValues {Curve {keys [<at 0 1>,<at 3.9 1>,<at 4.7 0.2>]}}}, CurveIndex=$Emitter.Age)
    Index 3 Param 3=&FloatFromCurve(FloatCurve=DataInterface() {propertyValues {Curve {keys [<at 0 1>,<at 3.9 1>,<at 4.7 0.2>]}}}, CurveIndex=$Emitter.Age)
    Module.Index0Float0Write=true
    Module.Index0Float1Write=true
    Module.Index0Float2Write=true
    Module.Index0Float3Write=true
    Module.Index1Float0Write=false
    Module.Index1Float1Write=false
    Module.Index1Float2Write=true
    Module.Index1Float3Write=false
    Module.Index3Float0Write=false
    Module.Index3Float1Write=false
    Module.Index3Float2Write=true
    Module.Index3Float3Write=false
    Write Parameter Index 1=::NewEnumerator1
    Write Parameter Index 3=::NewEnumerator1
   ) {module=DynamicMaterialParameters001}
  ]
  rendererClasses [&NiagaraRibbonRendererProperties]
  renderers [
   {rendererClass=&NiagaraRibbonRendererProperties}
   {rendererClass=&NiagaraRibbonRendererProperties, rendererIndex=1}
   {rendererClass=&NiagaraRibbonRendererProperties, rendererIndex=2}
  ]
 }
 DragonHead {
  emitterUpdateScript [
   &EmitterState(
    Inactive Response=::NewEnumerator1
    Life Cycle Mode=::NewEnumerator1
    Loop Behavior=::NewEnumerator1
    Loop Delay=0.2
    Loop Duration=4.4
    UseLoopDelay=true
   )
   &SpawnBurst_Instantaneous(Use Spawn Probability=false, UseLoopCountLimit=false)
  ]
  particleSpawnScript [
   &InitializeParticle(
    Color=<rgba 0.7 0.7 0.7 1>
    Lifetime=4.6
    Mesh Scale=&Multiply_VectorByFloat(Vector=<1 1 1>, Float=0.7)
    Mesh Scale Mode=::NewEnumerator4
    Position Offset=<-5 0 55>
    UsePositionOffset=true
    Write Material Random=true
   )
   &InitialMeshOrientation(Mesh Orientation Mode=::NewEnumerator2, Rotation=<0 0 0.25>, Use Rotation Vector=true)
   &InitialMeshOrientation(Mesh Orientation Mode=::NewEnumerator2, Rotation=<0 0 0.5>, Use Rotation Vector=true) {module=InitialMeshOrientation001}
   &InitialMeshOrientation(Mesh Orientation Mode=::NewEnumerator2, Use Rotation Vector=true) {module=InitialMeshOrientation002}
  ]
  particleUpdateScript [
   &ParticleState()
   &ScaleMeshSize(Scale Factor=&Multiply_VectorByFloat(
     Vector=<1 1 1>
     Float=&Multiply_Float(
      A=0.8
      B=&FloatFromCurve(
       FloatCurve=DataInterface() {propertyValues {Curve {
          defaultValue=1
          keys [{interpMode=RCIM_Cubic, tangentMode=RCTM_User, time=0, value=0}, {interpMode=RCIM_Cubic, tangentMode=RCTM_User, time=0.1, value=1}]
         }}}
       Scale Curve=&FloatFromCurve(FloatCurve=DataInterface() {propertyValues {Curve {keys [<at 0 1>,<at 3.9 1>,<at 4.7 0>]}}}, CurveIndex=$Emitter.Age)
      )
     )
    ))
   &Light_Attributes(
    Color=1
    Color Value To Scale:LinearColor=&MakeLinearColorFromVectorAndFloat("Vector (RGB)"=&MakeVectorFromLinearColorRGB(Color=<rgba 30 2.5 0 1>), "Float (Alpha)"=1)
    Enabled=1
    Falloff Exponent=1
    Light Exponent=3
    Light Radius=150
    Position=1
    Position Offset Coordinate Space:ENiagaraCoordinateSpace=::Local
    Radius=1
    Scale RGB:Vector3f=&Multiply_VectorByFloat(
     Vector=<1 1 1>
     Float=&FloatFromCurve(FloatCurve=DataInterface() {propertyValues {Curve {keys [<at 0.5 0>,<at 0.8 1>]}}}, CurveIndex=$Emitter.Age)
    )
    ScaleA=true
    ScaleRGB=true
    UsePositionOffset=true
   )
   &SkeletalMeshLocation(
    Bone Sampling Mode=::NewEnumerator4
    MaskSpawnGroup=false
    Skeletal Mesh=$System.Skeletal Mesh Reader
    Use Sampled Mesh Attribute as Ribbon ID=false
   ) {module=SkeletalMeshLocation001}
   &CameraOffset(Camera Offset Amount=&Add_Float(A=$System.CameraOffsetAll, B=10)) {module=CameraOffset001}
  ]
  rendererClasses [&NiagaraMeshRendererProperties, &NiagaraLightRendererProperties]
  renderers [
   {rendererClass=&NiagaraMeshRendererProperties, rendererIndex=0}
   {rendererClass=&NiagaraLightRendererProperties, rendererIndex=1}
  ]
 }
 DragonSmoke {
  emitterUpdateScript [
   &EmitterState(Inactive Response=::NewEnumerator1, Life Cycle Mode=::NewEnumerator1, Loop Behavior=::NewEnumerator1, Loop Duration=4.6, UseLoopDelay=false)
   &SpawnRate(SpawnRate=40, Use Spawn Probability=false)
  ]
  particleSpawnScript [
   &InitializeParticle(
    Color=&MakeCustomLinearColorFromBool(Bool=&RandomBool(), True Color=<rgba 3 0.223 0.0768 4>, False Color=<rgba 0.1 0 0 8>)
    Lifetime Max=0.8
    Lifetime Min=0.7
    Lifetime Mode=::NewEnumerator1
    Ribbon Width=40
    Ribbon Width Mode=::NewEnumerator1
    Sprite Rotation Mode=::NewEnumerator2
    Sprite Size Mode=::NewEnumerator1
    Uniform Sprite Size Max=120
    Uniform Sprite Size Min=70
    UsePositionOffset=false
    Write Material Random=true
   )
   &SkeletalMeshLocation(
    Bone Sampling Mode=::NewEnumerator4
    MaskSpawnGroup=false
    Skeletal Mesh=$System.Skeletal Mesh Reader
    Use Sampled Mesh Attribute as Ribbon ID=false
   ) {module=SkeletalMeshLocation001}
   &ShapeLocation(MaskSpawnGroup=false, Override Seed=false, Sphere Radius=10, UseOwnerScale=true)
  ]
  particleUpdateScript [
   &ParticleState()
   &SpriteRotationRate(Rotation Rate=&RandomRangeFloat(Minimum=-70, Maximum=70))
   &ScaleSpriteSize(Uniform Curve Sprite Scale=DataInterface() {propertyValues {Curve {
       defaultValue=0
       keys [
        {interpMode=RCIM_Cubic, tangentMode=RCTM_User, time=0, value=0.3}
        {arriveTangent=1, interpMode=RCIM_Cubic, leaveTangent=1, time=0.3001, value=1.0576}
        {interpMode=RCIM_Cubic, tangentMode=RCTM_User, time=1, value=1.3}
       ]
      }}})
   &DynamicMaterialParameters(
    Index 0 Param 1=&MakeCustomFloatFromBool(Bool=&RandomBool(), True Float=0, False Float=0.75)
    Module.Param0WriteEnabled=true
    Module.Param1WriteEnabled=true
    Module.Param2WriteEnabled=true
    Module.Param3WriteEnabled=true
   )
   &ScaleColor(
    Scale Alpha=&FloatFromCurve(
     FloatCurve=DataInterface() {propertyValues {Curve {keys [
         {interpMode=RCIM_Cubic, tangentMode=RCTM_User, time=0, value=0}
         {interpMode=RCIM_Cubic, tangentMode=RCTM_User, time=0.5, value=1}
         {tangentMode=RCTM_User, time=1, value=0}
        ]}}}
     Scale Curve=0.3
    )
    ScaleA=true
    ScaleRGB=true
   )
   &ScaleColor(
    Scale Alpha=&FN_FrontEndFloatModifier(Gameplay=1, Front End=0.5)
    Scale RGB=&MakeVectorFromLinearColorRGB(Color=&ColorFromCurve(ColorCurve=DataInterface() {propertyValues {
        AlphaCurve {keys []}
        BlueCurve {keys [<at 0 0.0021 RCIM_Cubic>,<at 0.345 0 RCIM_Cubic>]}
        GreenCurve {keys [<at 0 3.5705 RCIM_Cubic>,<at 0.345 0 RCIM_Cubic>]}
        RedCurve {keys [<at 0 5 RCIM_Cubic>,<at 0.345 1 RCIM_Cubic>]}
       }}))
    ScaleA=true
    ScaleRGB=true
   ) {module=ScaleColor001}
   &CurlNoiseForce(Bias Noise=false, Coordinate Space:ENiagaraCoordinateSpace=::Local, Noise Frequency=3, Noise Strength=100, Pan Noise=false, Random Seed=25)
   &SolveForcesAndVelocity(Clamp Velocity=false, Limit Acceleration=false)
   &CameraOffset(Camera Offset Amount=$System.CameraOffsetAll)
  ]
  rendererClasses [&NiagaraSpriteRendererProperties]
  renderers [{rendererClass=&NiagaraSpriteRendererProperties}]
 }
 Dragonsparks {
  emitterUpdateScript [
   &EmitterState(Inactive Response=::NewEnumerator1, Life Cycle Mode=::NewEnumerator1, Loop Behavior=::NewEnumerator1, Loop Duration=4.6, UseLoopDelay=false)
   &SpawnRate(SpawnRate=70, Use Spawn Probability=false)
  ]
  particleSpawnScript [
   &InitializeParticle(
    Color=<rgba 10 1.3868 1.3868 1>
    Lifetime Max=0.8
    Lifetime Min=0.7
    Lifetime Mode=::NewEnumerator1
    Ribbon Width=40
    Ribbon Width Mode=::NewEnumerator1
    Sprite Rotation Mode=::NewEnumerator2
    Sprite Size Max=<4 4>
    Sprite Size Min=<2 2>
    Sprite Size Mode=::NewEnumerator2
    UsePositionOffset=false
    Write Material Random=true
   )
   &SkeletalMeshLocation(
    Bone Sampling Mode=::NewEnumerator4
    MaskSpawnGroup=false
    Skeletal Mesh=$System.Skeletal Mesh Reader
    Use Sampled Mesh Attribute as Ribbon ID=false
   ) {module=SkeletalMeshLocation001}
   &ShapeLocation(MaskSpawnGroup=false, Override Seed=false, Sphere Radius=20, UseOwnerScale=true)
   &AddVelocity(
    Cone Angle=0
    Module.UseConeFalloff=false
    Module.UseVelDistribution=false
    Override Seed=false
    Velocity Mode=::NewEnumerator2
    Velocity Speed=&RandomRangeFloat(Minimum=50, Maximum=150)
   )
  ]
  particleUpdateScript [
   &ParticleState()
   &ScaleSpriteSize(Scale Factor=&Vector2DFromFloat(Value=&FloatFromCurve(
      FloatCurve=DataInterface() {propertyValues {Curve {keys [
          <at 0 0 RCIM_Cubic>
          {interpMode=RCIM_Cubic, leaveTangent=-3.0296, tangentMode=RCTM_Break, time=0.2, value=1.2}
          {arriveTangent=-1.8107, interpMode=RCIM_Cubic, leaveTangent=-1.8107, tangentMode=RCTM_User, time=1, value=0}
         ]}}}
      Scale Curve=0.4
     ))) {module=ScaleSpriteSize001}
   &SpriteRotationRate(Rotation Rate=&RandomRangeFloat(Minimum=-70, Maximum=70))
   /SetVariables(Particles.Glow03_Color=&MakeLinearColorFromVectorAndFloat(
     "Vector (RGB)"=&MakeVectorFromLinearColorRGB(Color=&MultiplyLinearColor(Linear Color=<rgba 1 0.0112 0.0112 1>, Scale Factor=&MakeVector4_FromFloat(Float=1)))
     "Float (Alpha)"=&Multiply_Float(A=0.2, B=1)
    )) {isSetParameters=true}
   /SetVariables(Particles.Glow03_Size=&Multiply_Vector2D_ByFloat(
     Float=&FloatFromCurve(
      FloatCurve=DataInterface() {propertyValues {Curve {keys [
          <at 0 0 RCIM_Cubic>
          <at 0.2 1.3 RCIM_Cubic>
          {arriveTangent=-0.1423, interpMode=RCIM_Cubic, leaveTangent=-0.1423, tangentMode=RCTM_User, time=1, value=0}
         ]}}}
      Scale Curve=8
     )
     Vector2D=<1 1>
    )) {isSetParameters=true}
   &DynamicMaterialParameters(
    Index 0 Param 1=&MakeCustomFloatFromBool(Bool=&RandomBool(), True Float=0, False Float=0.75)
    Module.Param0WriteEnabled=true
    Module.Param1WriteEnabled=true
    Module.Param2WriteEnabled=true
    Module.Param3WriteEnabled=true
   )
   &Drag(Use Linear Drag=true, Use Rotational Drag=true)
   &SolveForcesAndVelocity(Clamp Velocity=false, Limit Acceleration=false)
   &ScaleColor(
    Scale Alpha=&FloatFromCurve(FloatCurve=DataInterface() {propertyValues {Curve {keys [<at 0.3 1>,<at 0.4 1>,<at 1 1>]}}})
    ScaleA=true
    ScaleRGB=true
   ) {module=ScaleColor002}
   &CameraOffset(Camera Offset Amount=$System.CameraOffsetAll)
  ]
  rendererClasses [&NiagaraSpriteRendererProperties]
  renderers [
   {rendererClass=&NiagaraSpriteRendererProperties}
   {rendererClass=&NiagaraSpriteRendererProperties, rendererIndex=1}
  ]
 }
 HeadFire {
  emitterUpdateScript [
   &EmitterState(
    Inactive Response=::NewEnumerator1
    Life Cycle Mode=::NewEnumerator1
    Loop Behavior=::NewEnumerator1
    Loop Delay=0.2
    Loop Duration=4.4
    UseLoopDelay=true
   )
   &SpawnBurst_Instantaneous(Use Spawn Probability=false, UseLoopCountLimit=false)
  ]
  particleSpawnScript [
   &InitializeParticle(
    Color=<rgba 1 1 1 0.2>
    Lifetime=4.6
    Mesh Scale=&Multiply_Vector(
     A=&Multiply_Vector(A=<1 1 1.5>, B=$Particles.VectorScale)
     B=&VectorFromFloat(Value=$Particles.SizeScle)
    )
    Mesh Scale Mode=::NewEnumerator4
    UsePositionOffset=false
    Write Material Random=true
   ) {module=InitializeParticle001}
   /SetVariables(Particles.Mesh_Color2:LinearColor=&MakeLinearColorFromVectorAndFloat(
     "Vector (RGB)"=&MakeVectorFromLinearColorRGB(Color=&MultiplyLinearColor(Linear Color=<rgba 1 0.3113 0.2093 1>, Scale Factor=&MakeVector4_FromFloat(Float=1.2)))
     "Float (Alpha)"=0.3
    )) {isSetParameters=true}
   &CameraOffset() {enabled=false}
   &InitialMeshOrientation(Mesh Orientation Mode=::NewEnumerator2, Rotation=<0 0.25 0>, Use Rotation Vector=true)
   &InitialMeshOrientation(Mesh Orientation Mode=::NewEnumerator2, Rotation=<0.25 0 0>, Use Rotation Vector=true) {module=InitialMeshOrientation001}
  ]
  particleUpdateScript [
   /SetVariables(Particles.SizeScle:NiagaraFloat=&Multiply_Float(
     A=2.5
     B=&FloatFromCurve(FloatCurve=DataInterface() {propertyValues {Curve {keys [<at 0 1>,<at 3.9 1>,<at 4.7 0>]}}}, CurveIndex=$Emitter.Age)
    )) {isSetParameters=true}
   /SetVariables(Particles.VectorScale:Vector3f=<1.5 0.8 0.7>) {isSetParameters=true}
   &ParticleState()
   &ScaleMeshSize(Scale Factor=&Multiply_VectorByFloat(
     Vector=<1 1 1>
     Float=&FloatFromCurve(FloatCurve=DataInterface() {propertyValues {Curve {
         defaultValue=1
         keys [{interpMode=RCIM_Cubic, tangentMode=RCTM_User, time=0, value=0}, {interpMode=RCIM_Cubic, tangentMode=RCTM_User, time=0.1, value=1.0118}]
        }}})
    ))
   /SetVariables(Particles.Mesh_Size2:Vector3f=&Multiply_VectorByFloat(
     Vector=&Multiply_VectorByFloat(Vector=&Multiply_Vector(A=<1 0.9 1.1>, B=$Particles.VectorScale), Float=$Particles.SizeScle)
     Float=&FloatFromCurve(FloatCurve=DataInterface() {propertyValues {Curve {
         defaultValue=1
         keys [{interpMode=RCIM_Cubic, tangentMode=RCTM_User, time=0, value=0}, {interpMode=RCIM_Cubic, tangentMode=RCTM_User, time=0.5, value=1}]
        }}})
    )) {isSetParameters=true}
   &SkeletalMeshLocation(
    Bone Sampling Mode=::NewEnumerator4
    MaskSpawnGroup=false
    Sampled Position Offset=<10 0 0>
    Skeletal Mesh=$System.Skeletal Mesh Reader
    Use Sampled Mesh Attribute as Ribbon ID=false
   ) {module=SkeletalMeshLocation001}
   &InterpolateOverTime(Rate Of Change=20, Target Vector Value:NiagaraPosition=$Particles.Position, Variable Type=::NewEnumerator2) {module=InterpolateOverTime001}
   /SetVariables(Particles.Interpolate Dynamic=&MakeFloat4FromVectorAndFloat("Vector (RGB)"=$Particles.InterpolateOverTime001.InterpolateOverTime_V3_.Moving Average, "Float (Alpha)"=0)) {isSetParameters=true}
   &SolveForcesAndVelocity(Clamp Velocity=false, Limit Acceleration=false)
   &ScaleColor(
    Scale Alpha=&Multiply_Float(A=1, B=0.5)
    Scale RGB=&Multiply_VectorByFloat(
     Vector=<1 1 1>
     Float=&MakeCustomFloatFromBool(Bool=$User.bIsFrontEndPreview, True Float=0.5, False Float=2)
    )
    ScaleA=true
    ScaleRGB=true
   )
   &DynamicMaterialParameters(
    Index 0 Param 2=1
    Index 0 Param 3=2
    Index 0 Param 4=1
    Index 3 Vector 1:NiagaraPosition=$System.SampleSkeletalMesh.SampledPosition
    Module.Index0Float0Write=true
    Module.Index0Float1Write=true
    Module.Index0Float2Write=true
    Module.Index0Float3Write=true
    Module.Index3Float3Write=true
    Module.Index3VectorWrite=true
    Write Parameter Index 3=::NewEnumerator3
   ) {module=DynamicMaterialParameters001}
   &CameraOffset(Camera Offset Amount=$System.CameraOffsetAll) {module=CameraOffset001}
  ]
  rendererClasses [&NiagaraMeshRendererProperties]
  renderers [
   {rendererClass=&NiagaraMeshRendererProperties, rendererIndex=0}
   {rendererClass=&NiagaraMeshRendererProperties, rendererIndex=1}
  ]
 }
}
summary {
 systemName=NS_Emote_MasakoReign_Dragon
 emitters [
  {emitterName=Dragon, rendererClasses [&NiagaraRibbonRendererProperties]}
  {emitterName=DragonFireTrail, rendererClasses [&NiagaraRibbonRendererProperties]}
  {emitterName=DragonSmoke, rendererClasses [&NiagaraSpriteRendererProperties]}
  {emitterName=CurveFire, rendererClasses [&NiagaraMeshRendererProperties]}
  {emitterName=Dragonsparks, rendererClasses [&NiagaraSpriteRendererProperties]}
  {emitterName=DragonHead, rendererClasses [&NiagaraMeshRendererProperties, &NiagaraLightRendererProperties]}
  {emitterName=HeadFire, rendererClasses [&NiagaraMeshRendererProperties]}
 ]
 userVariables [
  User.bIsFrontEnd {type=NiagaraBool}
  User.bIsFrontEndPreview {type=NiagaraBool}
  User.ParaGlideLeanAlpha {type=NiagaraFloat, defaultValue=1, description=""}
  User.VariantSwitch {type=NiagaraFloat, defaultValue=0, description=""}
 ]
}
