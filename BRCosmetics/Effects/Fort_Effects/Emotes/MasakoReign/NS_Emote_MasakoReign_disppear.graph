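# Niagara effect for the MasakoReign emote, named as its disappear moment: a burst of fiery meshes and sparks offset to about torso height.
# CurveFire bursts spinning, randomly mirrored meshes outward, shaping size, brightness and alpha with curves, tinting from a system color curve and driving dynamic material parameters.
# Sparks bursts shrinking red sprites with a second glow size and color.
# Flash and Glow list no renderers; both curl noise modules are disabled.
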NiagaraSystem /BRCosmetics/Effects/Fort_Effects/Emotes/MasakoReign/NS_Emote_MasakoReign_disppear
dependencies {
 usedDataInterfaces [&NiagaraDataInterfaceCurve, &NiagaraDataInterfaceColorCurve]
 usedDynamicInputs [&FNTimeOfDayScalar]
 usedModules [
  &SystemState
  &self:SystemSpawnScript.SystemScriptSource.SystemScriptGraph.NiagaraNodeAssignment_1.SetVariables_#1
  &self:SystemSpawnScript.SystemScriptSource.SystemScriptGraph.NiagaraNodeAssignment_2.SetVariables_#2
  &self:SystemSpawnScript.SystemScriptSource.SystemScriptGraph.NiagaraNodeAssignment_3.SetVariables_#3
 ]
}
emitters {
 CurveFire {
  emitterUpdateScript [
   &EmitterState(Inactive Response=::NewEnumerator1, Life Cycle Mode=::NewEnumerator1, Loop Behavior=::NewEnumerator1, Loop Duration=4.6, UseLoopDelay=false)
   &SpawnBurst_Instantaneous(Spawn Count=18, Use Spawn Probability=true, UseLoopCountLimit=false)
   /SetVariables(Emitter.Allsize:NiagaraFloat=&RandomRangeFloat(Minimum=0.5, Maximum=0.9)) {isSetParameters=true}
  ]
  particleSpawnScript [
   &InitializeParticle(
    Color=<rgba 0.6 0.1466 0.1389 5>
    Lifetime Max=1.2
    Lifetime Min=0.5
    Lifetime Mode=::NewEnumerator1
    Mesh Scale=<0.5 0.5 0.5>
    Mesh Scale Mode=::NewEnumerator4
    Position Offset=<0 -20 140>
    UsePositionOffset=true
    Write Material Random=true
   )
   &ShapeLocation(
    Handle Radius=0
    Large Radius=30
    MaskSpawnGroup=false
    Override Seed=false
    Rotation Axis=<0 1 0>
    Rotation Mode=::NewEnumerator1
    Shape Primitive=::NewEnumerator3
    UseOwnerScale=true
   )
   &InitialMeshOrientation(Mesh Orientation Mode=::NewEnumerator7, Rotation=<0 0 0.25>, Use Rotation Vector=true)
   &InitialMeshOrientation(
    Mesh Orientation Mode=::NewEnumerator2
    Rotation=&RandomRangeVector(Minimum=<-1 0 0>, Maximum=<1 0 0>)
    Use Rotation Vector=true
   ) {module=InitialMeshOrientation001}
   &AddVelocity(
    Module.Use Falloff Distance=false
    Origin Offset=<0 -20 140>
    Override Seed=false
    Velocity Mode=::NewEnumerator1
    Velocity Speed=&Multiply_Float(A=200, B=&RandomRangeFloat(Minimum=0.2, Maximum=1))
   )
   &CameraOffset(Camera Offset Amount=-20)
  ]
  particleUpdateScript [
   /SetVariables(Particles.AlphaAll:NiagaraFloat=&FloatFromCurve(
     FloatCurve=DataInterface() {propertyValues {Curve {defaultValue=1, keys [{tangentMode=RCTM_User, time=0, value=0}, {tangentMode=RCTM_User, time=0.5, value=1}]}}}
     Scale Curve=&FloatFromCurve(
      FloatCurve=DataInterface() {propertyValues {Curve {keys [{arriveTangent=-2.8607, interpMode=RCIM_Cubic, leaveTangent=-2.8607, tangentMode=RCTM_User, time=0.0044, value=0.9529}, <at 1 0>]}}}
      Scale Curve=&Multiply_Float(A=1.5, B=1)
     )
    )) {isSetParameters=true}
   /SetVariables(Particles.BrightnessAll:NiagaraFloat=&FloatFromCurve(
     FloatCurve=DataInterface() {propertyValues {Curve {keys [
         <at 0 -0.0118 RCIM_Cubic>
         {arriveTangent=0.3618, interpMode=RCIM_Cubic, leaveTangent=0.3618, time=0.2035, value=1}
         {arriveTangent=-1.2554, interpMode=RCIM_Cubic, leaveTangent=-1.2554, time=0.5284, value=0.1794}
         <at 1 0>
        ]}}}
     Scale Curve=&MakeCustomFloatFromBool(Bool=&RandomBool(), True Float=0.3, False Float=1)
    )) {isSetParameters=true}
   &DynamicMaterialParameters(
    Index 0 Param 1=&Multiply_Float(A=1.1, B=$Emitter.Allsize)
    Index 0 Param 2=2
    Index 0 Param 3=&RandomRangeFloat(Minimum=0, Maximum=1)
    Index 0 Param 4=0.5
    Index 1 Param 1=&FloatFromCurve(FloatCurve=DataInterface() {propertyValues {Curve {keys [<at 0 0>,<at 1 1>]}}})
    Index 1 Param 2=0.3
    Module.Index0Float0Write=true
    Module.Index0Float1Write=true
    Module.Index0Float2Write=true
    Module.Index0Float3Write=true
    Module.Index1Float0Write=true
    Module.Index1Float1Write=true
    Module.Index1Float2Write=false
    Module.Index1Float3Write=false
    Write Parameter Index 1=::NewEnumerator1
   )
   &ParticleState()
   &UpdateMeshOrientation(Rotation Rate=-0.2, Rotation Vector=<0 1 0>)
   &ScaleMeshSize(Scale Factor=&Multiply_VectorByFloat(
     Vector=&MakeCustomVectorFromBool(Bool=&RandomBool(), True Vector=<-1 1 1>, False Vector=<1 1 1>)
     Float=&FloatFromCurve(
      FloatCurve=DataInterface() {propertyValues {Curve {
         defaultValue=1
         keys [
          {arriveTangent=1.0428, interpMode=RCIM_Cubic, time=0, value=0}
          {arriveTangent=-0.2, interpMode=RCIM_Cubic, leaveTangent=-0.2, tangentMode=RCTM_Break, time=0.1997, value=1.1761}
          {interpMode=RCIM_Cubic, leaveTangent=0.8255, time=1, value=1}
         ]
        }}}
      Scale Curve=&Multiply_Float(A=&RandomRangeFloat(Minimum=0.8, Maximum=1.2), B=$Emitter.Allsize)
     )
    ))
   &Drag(Drag=2, Use Linear Drag=true, Use Rotational Drag=true)
   &CurlNoiseForce(Bias Noise=false, Coordinate Space:ENiagaraCoordinateSpace=::Local, Noise Frequency=30, Noise Strength=500, Pan Noise=false, Random Seed=25) {enabled=false}
   &SolveForcesAndVelocity(Clamp Velocity=false, Limit Acceleration=false)
   &ScaleColor(
    Scale Alpha=$Particles.AlphaAll
    Scale RGB=&Multiply_VectorByFloat(
     Vector=&MakeVectorFromLinearColorRGB(Color=&ColorFromCurve(ColorCurve=$System.Curve for Colors))
     Float=$Particles.BrightnessAll
    )
    ScaleA=true
    ScaleRGB=true
   )
   &FN_Mesh_Attributes(
    Color=&MakeLinearColorFromVectorAndFloat(
     "Vector (RGB)"=&Multiply_Vector(
      A=&Multiply_VectorByFloat(Vector=&MakeVectorFromLinearColorRGB(Color=<rgba 5 0.4167 0 0.3>), Float=$Particles.BrightnessAll)
      B=&MakeVectorFromLinearColorRGB(Color=&ColorFromCurve(ColorCurve=$System.Curve for Colors))
     )
     "Float (Alpha)"=&Multiply_Float(A=0.3, B=$Particles.AlphaAll)
    )
    Create Color Attribute=true
    Module Output=::NewEnumerator1
   )
  ]
  rendererClasses [&NiagaraMeshRendererProperties]
  renderers [
   {rendererClass=&NiagaraMeshRendererProperties, rendererIndex=0}
   {rendererClass=&NiagaraMeshRendererProperties, rendererIndex=1}
  ]
 }
 Flash {rendererClasses [], renderers []}
 Glow {rendererClasses [], renderers []}
 Sparks {
  emitterUpdateScript [
   &EmitterState(Inactive Response=::NewEnumerator1, Life Cycle Mode=::NewEnumerator1, Loop Behavior=::NewEnumerator1, Loop Duration=4.6, UseLoopDelay=false)
   &SpawnBurst_Instantaneous(Spawn Count=30, Use Spawn Probability=false, UseLoopCountLimit=false)
  ]
  particleSpawnScript [
   &InitializeParticle(
    Color=<rgba 10 1.3868 1.3868 1>
    Lifetime Max=0.6
    Lifetime Min=0.7
    Lifetime Mode=::NewEnumerator1
    Position Offset=<0 -20 140>
    Ribbon Width=40
    Ribbon Width Mode=::NewEnumerator1
    Sprite Rotation Mode=::NewEnumerator2
    Sprite Size Max=<4 4>
    Sprite Size Min=<2 2>
    Sprite Size Mode=::NewEnumerator2
    UsePositionOffset=true
    Write Material Random=true
   )
   &ShapeLocation(
    Handle Radius=0
    Large Radius=20
    MaskSpawnGroup=false
    Override Seed=false
    Rotation Axis=<0 1 0>
    Rotation Mode=::NewEnumerator1
    Shape Primitive=::NewEnumerator3
    Surface Distribution=0.5
    UseOwnerScale=true
   )
   &AddVelocity(
    Module.Use Falloff Distance=false
    Origin Offset=<0 -20 140>
    Override Seed=false
    Velocity Mode=::NewEnumerator1
    Velocity Speed=&RandomRangeFloat(Minimum=70, Maximum=150)
   )
  ]
  particleUpdateScript [
   &ParticleState()
   &ScaleSpriteSize(Scale Factor=&Vector2DFromFloat(Value=&FloatFromCurve(
      FloatCurve=DataInterface() {propertyValues {Curve {keys [
          <at 0 0 RCIM_Cubic>
          {interpMode=RCIM_Cubic, leaveTangent=-3.0296, tangentMode=RCTM_Break, time=0.2, value=1.2}
          {arriveTangent=-1.8107, interpMode=RCIM_Cubic, leaveTangent=-1.8107, tangentMode=RCTM_User, time=1, value=0}
         ]}}}
      Scale Curve=0.4
     ))) {module=ScaleSpriteSize001}
   &SpriteRotationRate(Rotation Rate=&RandomRangeFloat(Minimum=-70, Maximum=70))
   /SetVariables(Particles.Glow03_Color=&MakeLinearColorFromVectorAndFloat(
     "Vector (RGB)"=&MakeVectorFromLinearColorRGB(Color=&MultiplyLinearColor(Linear Color=<rgba 1 0.0112 0.0112 1>, Scale Factor=&MakeVector4_FromFloat(Float=1)))
     "Float (Alpha)"=&Multiply_Float(A=0.2, B=1)
    )) {isSetParameters=true}
   /SetVariables(Particles.Glow03_Size=&Multiply_Vector2D_ByFloat(
     Float=&FloatFromCurve(
      FloatCurve=DataInterface() {propertyValues {Curve {keys [
          <at 0 0 RCIM_Cubic>
          <at 0.2 1.3 RCIM_Cubic>
          {arriveTangent=-0.1423, interpMode=RCIM_Cubic, leaveTangent=-0.1423, tangentMode=RCTM_User, time=1, value=0}
         ]}}}
      Scale Curve=8
     )
     Vector2D=<1 1>
    )) {isSetParameters=true}
   &DynamicMaterialParameters(
    Index 0 Param 1=&MakeCustomFloatFromBool(Bool=&RandomBool(), True Float=0, False Float=0.75)
    Module.Param0WriteEnabled=true
    Module.Param1WriteEnabled=true
    Module.Param2WriteEnabled=true
    Module.Param3WriteEnabled=true
   )
   &CurlNoiseForce(Bias Noise=false, Coordinate Space:ENiagaraCoordinateSpace=::Local, Noise Strength=500, Pan Noise=false, Random Seed=25) {enabled=false}
   &Drag(Use Linear Drag=true, Use Rotational Drag=true)
   &SolveForcesAndVelocity(Clamp Velocity=false, Limit Acceleration=false)
   &ScaleColor(
    Scale Alpha=&FloatFromCurve(FloatCurve=DataInterface() {propertyValues {Curve {keys [<at 0.3 1>,<at 0.4 1>,<at 1 1>]}}})
    ScaleA=true
    ScaleRGB=true
   ) {module=ScaleColor002}
  ]
  rendererClasses [&NiagaraSpriteRendererProperties]
  renderers [
   {rendererClass=&NiagaraSpriteRendererProperties}
   {rendererClass=&NiagaraSpriteRendererProperties, rendererIndex=1}
  ]
 }
}
summary {
 systemName=NS_Emote_MasakoReign_disppear
 emitters [
  {emitterName=Flash, rendererClasses []}
  {emitterName=Glow, rendererClasses []}
  {emitterName=CurveFire, rendererClasses [&NiagaraMeshRendererProperties]}
  {emitterName=Sparks, rendererClasses [&NiagaraSpriteRendererProperties]}
 ]
 userVariables [
  User.bIsFrontEnd {type=NiagaraBool}
  User.bIsFrontEndPreview {type=NiagaraBool}
  User.ParaGlideLeanAlpha {type=NiagaraFloat, defaultValue=1, description=""}
  User.VariantSwitch {type=NiagaraFloat, defaultValue=0, description=""}
 ]
}
